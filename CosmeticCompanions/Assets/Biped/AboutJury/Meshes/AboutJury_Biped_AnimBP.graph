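# Animation Blueprint for the AboutJury biped cosmetic companion, whose output pose is retargeted from another mesh.
# bUseReverseRetarget drives a cubic-eased blend between two Retarget Pose From Mesh nodes: the emote retargeter when true, the standard one when false.
# A separate bowtie constraint chain, switching direction between bowtie and emote attach bones on that flag, does not feed the output pose.
# The event graph is empty.
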AnimBlueprint /CosmeticCompanions/Assets/Biped/AboutJury/Meshes/AboutJury_Biped_AnimBP
target_skeleton=&/AboutJury_Biped_Skeleton
graphs [
 AnimGraph {nodes [
   Component To Local()
   n1 = "Constraint - bowtie_root (PPPP)"!(
    ComponentPose=Local To Component()
    Alpha=1
    ConstraintWeights_0=0.6
    ConstraintWeights_1:float=0.1
    ConstraintWeights_2:float=0.2
    ConstraintWeights_3:float=0.1
   ) -> (Pose) {
    boneToModify {boneName=bowtie_root}
    constraintSetup [{targetBone {boneName=neck_02}}, {targetBone {boneName=spine_05}}, {targetBone {boneName=neck_01}}, {targetBone {boneName=head}}]
    constraintWeights [1,1,1,1]
   }
   n2 = Get bUseReverseRetarget() -> (bUseReverseRetarget:bool)
   n3 = "Constraint - attach_emote_01 (P)"!(ComponentPose=@n1, bAlphaBoolEnabled:bool=NOT Boolean(A=@n2), ConstraintWeights_0=1) -> (Pose) {
    alphaInputType=Bool
    boneToModify {boneName=attach_emote_01}
    constraintSetup [{offsetOption=None, targetBone {boneName=bowtie_root}}]
    constraintWeights [1]
   }
   n4 = "Constraint - bowtie (P)"!(ComponentPose=@n3, bAlphaBoolEnabled:bool=@n2, ConstraintWeights_0=1) -> (Pose) {
    alphaInputType=Bool
    boneToModify {boneName=bowtie}
    constraintSetup [{offsetOption=None, targetBone {boneName=attach_emote_01}}]
    constraintWeights [1]
   }
   Component To Local(ComponentPose=@n4)
   Copy Pose From Mesh() {bCopyCurves=true, bUseAttachedParent=true}
   Local To Component()
   Output Pose(Result=Blend Poses by bool(
     BlendPose_0=Retarget Pose From Mesh!() -> (Pose:Pose Link Structure) {is=title AnimGraphNode_RetargetPoseFromMesh, iKRetargeterAsset=&/RTG_AboutJury_Emote}
     BlendPose_1=Retarget Pose From Mesh() {iKRetargeterAsset=&/RTG_AboutJury}
     BlendTime_0=0.5
     BlendTime_1=0.3
     bActiveValue=Get bUseReverseRetarget()
    ) {blendType=CubicInOut})
  ]}
 EventGraph {links [], nodes []}
]
variables []
