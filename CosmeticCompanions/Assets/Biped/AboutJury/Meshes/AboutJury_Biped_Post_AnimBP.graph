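# Post-process AnimBP for the AboutJury biped cosmetic companion, adding secondary motion to ears, whiskers, hair and bowtie.
# A Pose Blend node applies the PoseAsset to the input pose, ignoring the asset's source animation curves.
# Serial single-bone AnimDynamics nodes add root-relative, angular-sprung jiggle with reduced gravity and tight limits to ears, whiskers and hair.
# A weighted constraint then binds bowtie_root to neck, upper-spine and head bones, before a small additive bone-space bowtie offset.
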AnimBlueprint /CosmeticCompanions/Assets/Biped/AboutJury/Meshes/AboutJury_Biped_Post_AnimBP
target_skeleton=&/AboutJury_Biped_Skeleton
graphs [
 AnimGraph {nodes [
   n1 = Input Pose()
   n2 = Get PoseAsset() -> (PoseAsset:Pose Asset Object Reference)
   n3 = Anim Dynamics - Bone: ear_01_l(
    ComponentPose=Local To Component(LocalPose=Pose!(PoseAsset:Pose Asset Object Reference=@n2, SourcePose:Pose Link Structure=@n1) -> (Pose:Pose Link Structure) {is=title AnimGraphNode_PoseBlendNode, bIgnorePoseAssetSourceAnimationCurves=true, blendOption=Linear})
    Alpha=1
    ExternalForce="0,0,0"
   ) {
    angularDampingOverride=1
    angularSpringConstant=30000
    bAngularSpring=true
    bOverrideAngularDamping=true
    bUseGravityOverride=true
    gravityScale=0.35
    simulationSpace=RootRelative
    alphaScaleBiasClamp {bias=1, scale=-1}
    boundBone {boneName=ear_01_l}
    gravityOverride {z=-100}
    physicsBodyDefinitions [{
      boundBone {boneName=ear_01_l}
      boxExtents=<20 30 30>
      constraintSetup {angularLimitsMax=<3 3 3>, angularLimitsMin=<-3 -3 -3>, angularTarget=<1 0 0>}
      localJointOffset=<2 0 0>
     }]
    simSpaceSettings {maxAngularAcceleration=10, maxAngularVelocity=300, simSpaceAngularAlpha=1}
   }
   n4 = Anim Dynamics - Bone: ear_02_l(ComponentPose=@n3, Alpha=1, ExternalForce="0,0,0") {
    angularDampingOverride=1
    angularSpringConstant=30000
    bAngularSpring=true
    bOverrideAngularDamping=true
    bUseGravityOverride=true
    gravityScale=0.35
    simulationSpace=RootRelative
    alphaScaleBiasClamp {bias=1, scale=-1}
    boundBone {boneName=ear_02_l}
    gravityOverride {z=-100}
    physicsBodyDefinitions [{
      boundBone {boneName=ear_02_l}
      boxExtents=<20 30 30>
      constraintSetup {angularLimitsMax=<5 5 5>, angularLimitsMin=<-5 -5 -5>, angularTarget=<1 0 0>}
      localJointOffset=<2 0 0>
     }]
    simSpaceSettings {maxAngularAcceleration=10, maxAngularVelocity=300, simSpaceAngularAlpha=1}
   }
   n5 = Anim Dynamics - Bone: ear_03_l(ComponentPose=@n4, Alpha=1, ExternalForce="0,0,0") {
    angularDampingOverride=1
    angularSpringConstant=30000
    bAngularSpring=true
    bOverrideAngularDamping=true
    bUseGravityOverride=true
    gravityScale=0.35
    simulationSpace=RootRelative
    alphaScaleBiasClamp {bias=1, scale=-1}
    boundBone {boneName=ear_03_l}
    gravityOverride {z=-100}
    physicsBodyDefinitions [{
      boundBone {boneName=ear_03_l}
      boxExtents=<20 30 30>
      constraintSetup {angularLimitsMax=<7 7 7>, angularLimitsMin=<-7 -7 -7>, angularTarget=<1 0 0>}
      localJointOffset=<2 0 0>
     }]
    simSpaceSettings {maxAngularAcceleration=10, maxAngularVelocity=300, simSpaceAngularAlpha=1}
   }
   n6 = Anim Dynamics - Bone: ear_01_r(ComponentPose=@n5, Alpha=1, ExternalForce="0,0,0") {
    angularDampingOverride=1
    angularSpringConstant=30000
    bAngularSpring=true
    bOverrideAngularDamping=true
    bUseGravityOverride=true
    gravityScale=0.35
    simulationSpace=RootRelative
    alphaScaleBiasClamp {bias=1, scale=-1}
    boundBone {boneName=ear_01_r}
    gravityOverride {z=-100}
    physicsBodyDefinitions [{
      boundBone {boneName=ear_01_r}
      boxExtents=<20 30 30>
      constraintSetup {angularLimitsMax=<3 3 3>, angularLimitsMin=<-3 -3 -3>, angularTarget=<1 0 0>}
      localJointOffset=<-2 0 0>
     }]
    simSpaceSettings {maxAngularAcceleration=10, maxAngularVelocity=300, simSpaceAngularAlpha=1}
   }
   n7 = Anim Dynamics - Bone: ear_02_r(ComponentPose=@n6, Alpha=1, ExternalForce="0,0,0") {
    angularDampingOverride=1
    angularSpringConstant=30000
    bAngularSpring=true
    bOverrideAngularDamping=true
    bUseGravityOverride=true
    gravityScale=0.35
    simulationSpace=RootRelative
    alphaScaleBiasClamp {bias=1, scale=-1}
    boundBone {boneName=ear_02_r}
    gravityOverride {z=-100}
    physicsBodyDefinitions [{
      boundBone {boneName=ear_02_r}
      boxExtents=<20 30 30>
      constraintSetup {angularLimitsMax=<5 5 5>, angularLimitsMin=<-5 -5 -5>, angularTarget=<1 0 0>}
      localJointOffset=<-2 0 0>
     }]
    simSpaceSettings {maxAngularAcceleration=10, maxAngularVelocity=300, simSpaceAngularAlpha=1}
   }
   n8 = Anim Dynamics - Bone: ear_03_r(ComponentPose=@n7, Alpha=1, ExternalForce="0,0,0") {
    angularDampingOverride=1
    angularSpringConstant=30000
    bAngularSpring=true
    bOverrideAngularDamping=true
    bUseGravityOverride=true
    gravityScale=0.35
    simulationSpace=RootRelative
    alphaScaleBiasClamp {bias=1, scale=-1}
    boundBone {boneName=ear_03_r}
    gravityOverride {z=-100}
    physicsBodyDefinitions [{
      boundBone {boneName=ear_03_r}
      boxExtents=<20 30 30>
      constraintSetup {angularLimitsMax=<7 7 7>, angularLimitsMin=<-7 -7 -7>, angularTarget=<1 0 0>}
      localJointOffset=<-2 0 0>
     }]
    simSpaceSettings {maxAngularAcceleration=10, maxAngularVelocity=300, simSpaceAngularAlpha=1}
   }
   n9 = Anim Dynamics - Bone: whisker_01_l(ComponentPose=@n8, Alpha=1, ExternalForce="0,0,0") {
    angularDampingOverride=1
    angularSpringConstant=30000
    bAngularSpring=true
    bOverrideAngularDamping=true
    bUseGravityOverride=true
    gravityScale=0.35
    simulationSpace=RootRelative
    alphaScaleBiasClamp {bias=1, scale=-1}
    boundBone {boneName=whisker_01_l}
    gravityOverride {z=-100}
    physicsBodyDefinitions [{
      boundBone {boneName=whisker_01_l}
      boxExtents=<20 20 30>
      constraintSetup {angularLimitsMax=<15 15 15>, angularLimitsMin=<-15 -15 -15>, angularTarget=<1 0 0>}
      localJointOffset=<4 0 0>
     }]
    simSpaceSettings {maxAngularAcceleration=10, maxAngularVelocity=300, simSpaceAngularAlpha=1}
   }
   n10 = Anim Dynamics - Bone: whisker_02_l(ComponentPose=@n9, Alpha=1, ExternalForce="0,0,0") {
    angularDampingOverride=1
    angularSpringConstant=30000
    bAngularSpring=true
    bOverrideAngularDamping=true
    bUseGravityOverride=true
    gravityScale=0.35
    simulationSpace=RootRelative
    alphaScaleBiasClamp {bias=1, scale=-1}
    boundBone {boneName=whisker_02_l}
    gravityOverride {z=-100}
    physicsBodyDefinitions [{
      boundBone {boneName=whisker_02_l}
      boxExtents=<30 20 20>
      constraintSetup {angularLimitsMax=<15 15 15>, angularLimitsMin=<-15 -15 -15>, angularTarget=<1 0 0>}
      localJointOffset=<4 0 0>
     }]
    simSpaceSettings {maxAngularAcceleration=10, maxAngularVelocity=300, simSpaceAngularAlpha=1}
   }
   n11 = Anim Dynamics - Bone: whisker_03_l(ComponentPose=@n10, Alpha=1, ExternalForce="0,0,0") {
    angularDampingOverride=1
    angularSpringConstant=30000
    bAngularSpring=true
    bOverrideAngularDamping=true
    bUseGravityOverride=true
    gravityScale=0.35
    simulationSpace=RootRelative
    alphaScaleBiasClamp {bias=1, scale=-1}
    boundBone {boneName=whisker_03_l}
    gravityOverride {z=-100}
    physicsBodyDefinitions [{
      boundBone {boneName=whisker_03_l}
      boxExtents=<20 30 20>
      constraintSetup {angularLimitsMax=<15 15 15>, angularLimitsMin=<-15 -15 -15>, angularTarget=<1 0 0>}
      localJointOffset=<4 0 0>
     }]
    simSpaceSettings {maxAngularAcceleration=10, maxAngularVelocity=300, simSpaceAngularAlpha=1}
   }
   n12 = Anim Dynamics - Bone: whisker_01_r(ComponentPose=@n11, Alpha=1, ExternalForce="0,0,0") {
    angularDampingOverride=1
    angularSpringConstant=30000
    bAngularSpring=true
    bOverrideAngularDamping=true
    bUseGravityOverride=true
    gravityScale=0.35
    simulationSpace=RootRelative
    alphaScaleBiasClamp {bias=1, scale=-1}
    boundBone {boneName=whisker_01_r}
    gravityOverride {z=-100}
    physicsBodyDefinitions [{
      boundBone {boneName=whisker_01_r}
      boxExtents=<20 30 30>
      constraintSetup {angularLimitsMax=<15 15 15>, angularLimitsMin=<-15 -15 -15>, angularTarget=<1 0 0>}
      localJointOffset=<-4 0 0>
     }]
    simSpaceSettings {maxAngularAcceleration=10, maxAngularVelocity=300, simSpaceAngularAlpha=1}
   }
   n13 = Anim Dynamics - Bone: whisker_02_r(ComponentPose=@n12, Alpha=1, ExternalForce="0,0,0") {
    angularDampingOverride=1
    angularSpringConstant=30000
    bAngularSpring=true
    bOverrideAngularDamping=true
    bUseGravityOverride=true
    gravityScale=0.35
    simulationSpace=RootRelative
    alphaScaleBiasClamp {bias=1, scale=-1}
    boundBone {boneName=whisker_02_r}
    gravityOverride {z=-100}
    physicsBodyDefinitions [{
      boundBone {boneName=whisker_02_r}
      boxExtents=<30 20 20>
      constraintSetup {angularLimitsMax=<15 15 15>, angularLimitsMin=<-15 -15 -15>, angularTarget=<1 0 0>}
      localJointOffset=<-4 0 0>
     }]
    simSpaceSettings {maxAngularAcceleration=10, maxAngularVelocity=300, simSpaceAngularAlpha=1}
   }
   n14 = Anim Dynamics - Bone: whisker_03_r(ComponentPose=@n13, Alpha=1, ExternalForce="0,0,0") {
    angularDampingOverride=1
    angularSpringConstant=30000
    bAngularSpring=true
    bOverrideAngularDamping=true
    bUseGravityOverride=true
    gravityScale=0.35
    simulationSpace=RootRelative
    alphaScaleBiasClamp {bias=1, scale=-1}
    boundBone {boneName=whisker_03_r}
    gravityOverride {z=-100}
    physicsBodyDefinitions [{
      boundBone {boneName=whisker_03_r}
      boxExtents=<30 30 20>
      constraintSetup {angularLimitsMax=<15 15 15>, angularLimitsMin=<-15 -15 -15>, angularTarget=<1 0 0>}
      localJointOffset=<-4 0 0>
     }]
    simSpaceSettings {maxAngularAcceleration=10, maxAngularVelocity=300, simSpaceAngularAlpha=1}
   }
   n15 = Anim Dynamics - Bone: hair_01(ComponentPose=@n14, Alpha=1, ExternalForce="0,0,0") {
    angularDampingOverride=1
    angularSpringConstant=10000
    bAngularSpring=true
    bOverrideAngularDamping=true
    bUseGravityOverride=true
    gravityScale=0.35
    simulationSpace=RootRelative
    alphaScaleBiasClamp {bias=1, scale=-1}
    boundBone {boneName=hair_01}
    gravityOverride {z=-50}
    physicsBodyDefinitions [{
      boundBone {boneName=hair_01}
      boxExtents=<25 15 15>
      constraintSetup {angularLimitsMax=<5 5 5>, angularLimitsMin=<-5 -5 -5>, angularTarget=<1 0 0>}
      localJointOffset=<2 0 0>
     }]
    simSpaceSettings {maxAngularAcceleration=10, maxAngularVelocity=300, simSpaceAngularAlpha=1}
   }
   n16 = Anim Dynamics - Bone: hair_02(ComponentPose=@n15, Alpha=1, ExternalForce="0,0,0") {
    angularDampingOverride=1
    angularSpringConstant=10000
    bAngularSpring=true
    bOverrideAngularDamping=true
    bUseGravityOverride=true
    gravityScale=0.35
    simulationSpace=RootRelative
    alphaScaleBiasClamp {bias=1, scale=-1}
    boundBone {boneName=hair_02}
    gravityOverride {z=-50}
    physicsBodyDefinitions [{
      boundBone {boneName=hair_02}
      boxExtents=<25 15 15>
      constraintSetup {angularLimitsMax=<5 5 5>, angularLimitsMin=<-5 -5 -5>, angularTarget=<1 0 0>}
      localJointOffset=<2 0 0>
     }]
    simSpaceSettings {maxAngularAcceleration=10, maxAngularVelocity=300, simSpaceAngularAlpha=1}
   }
   n17 = Anim Dynamics - Bone: hair_03(ComponentPose=@n16, Alpha=1, ExternalForce="0,0,0") {
    angularDampingOverride=1
    angularSpringConstant=10000
    bAngularSpring=true
    bOverrideAngularDamping=true
    bUseGravityOverride=true
    gravityScale=0.35
    simulationSpace=RootRelative
    alphaScaleBiasClamp {bias=1, scale=-1}
    boundBone {boneName=hair_03}
    gravityOverride {z=-50}
    physicsBodyDefinitions [{
      boundBone {boneName=hair_03}
      boxExtents=<25 15 15>
      constraintSetup {angularLimitsMax=<5 5 5>, angularLimitsMin=<-5 -5 -5>, angularTarget=<1 0 0>}
      localJointOffset=<2 0 0>
     }]
    simSpaceSettings {maxAngularAcceleration=10, maxAngularVelocity=300, simSpaceAngularAlpha=1}
   }
   n18 = Anim Dynamics - Bone: hair_04(ComponentPose=@n17, Alpha=1, ExternalForce="0,0,0") {
    angularDampingOverride=1
    angularSpringConstant=10000
    bAngularSpring=true
    bOverrideAngularDamping=true
    bUseGravityOverride=true
    gravityScale=0.35
    simulationSpace=RootRelative
    alphaScaleBiasClamp {bias=1, scale=-1}
    boundBone {boneName=hair_04}
    gravityOverride {z=-50}
    physicsBodyDefinitions [{
      boundBone {boneName=hair_04}
      boxExtents=<25 15 15>
      constraintSetup {angularLimitsMax=<4 4 4>, angularLimitsMin=<-4 -4 -4>, angularTarget=<1 0 0>}
      localJointOffset=<2 0 0>
     }]
    simSpaceSettings {maxAngularAcceleration=10, maxAngularVelocity=300, simSpaceAngularAlpha=1}
   }
   n19 = "Constraint - bowtie_root (PPPP)"!(
    ComponentPose=@n18
    Alpha=1
    ConstraintWeights_0=0.2
    ConstraintWeights_1:float=0.4
    ConstraintWeights_2:float=0.2
    ConstraintWeights_3:float=0.2
   ) -> (Pose) {
    boneToModify {boneName=bowtie_root}
    constraintSetup [{targetBone {boneName=neck_02}}, {targetBone {boneName=spine_05}}, {targetBone {boneName=neck_01}}, {targetBone {boneName=head}}]
    constraintWeights [1,1,1,1]
   }
   n20 = "Transform (Modify) Bone - Bone: bowtie"(ComponentPose=@n19, Alpha=1, Translation="-0.4,0.2,0", Rotation="0,0,0", Scale="1,1,1") {translationMode=BMM_Additive, translationSpace=BCS_BoneSpace, boneToModify {boneName=bowtie}, translation=<-0.4 0.2>}
   Output Pose(Result=Component To Local(ComponentPose=@n20))
  ]}
 EventGraph {links [], nodes []}
]
variables [PoseAsset {type=PoseAsset, default=&/AboutJury_Biped_PoseAsset}]
